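# Niagara vehicle boost cosmetic with an elemental theme, combining flying rocks, water and fire rings, two-color ribbon trails and orbiting particles.
# Four emitters are placed using user boost-position offsets, with socket filtering and paint overrides disabled.
# Local_Particles001 merges three particle types through spawn groups, switching color, size, orbit and fade per group.
# Ribbons pack extra binding colors across several renderers and adjust color intensity for frontend preview.
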NiagaraSystem /VehicleCosmetics/Boosts/Boost_Elements/FX/NS_Boost_Elements_Optimized
dependencies {
 usedDataInterfaces [
  &NiagaraDataInterfaceCurve
  &NiagaraDataInterfaceVectorCurve
  &NiagaraDataInterfaceArrayColor
  &NiagaraDataInterfaceSkeletalMesh
 ]
 usedDynamicInputs [&Multiply_Vector2D_ByFloat]
 usedModules [
  &self:SystemSpawnScript.SystemScriptSource.SystemScriptGraph.NiagaraNodeAssignment_0.SetVariables_#1
  &SystemState
 ]
}
emitters {
 Flying_Rocks001 {
  emitterUpdateScript [
   &EmitterState()
   &VC_FilteredSocketCount() {enabled=false}
   &SpawnRate(SpawnRate=30, Use Spawn Probability=false)
  ]
  particleSpawnScript [
   &VC_PaintedVariantsOverride() {enabled=false}
   &InitializeParticle(
    Color=<rgba 0.2396 0.1214 0 1>
    Lifetime Max=&Multiply_Float(A=0.4, B=0.4)
    Lifetime Min:NiagaraFloat=&Multiply_Float(A=0.7, B=0.4)
    Lifetime Mode=::NewEnumerator1
    Mesh Scale Max=&Multiply_VectorByFloat(Vector=<0.2 0.2 0.2>, Float=3)
    Mesh Scale Min=&Multiply_VectorByFloat(Vector=<0.1 0.1 0.1>, Float=3)
    Mesh Scale Mode=::NewEnumerator2
    Position Offset=&AddVectorToPosition(
     Position=&AddVectorToPosition(
      Position=<0 0 30>
      Vector=&AddVectorToPosition(Position=$User.DTCarOffset, Vector=$User.SingleBoostPos)
     )
     Vector=<-125 0 -30>
    )
    Position Offset Coordinate Space=::Local
    UsePositionOffset=true
    Write Material Random=true
   )
   &PartitionParticles(Number Of Partitions=$Emitter.PortCount) {enabled=false}
   &SkeletalMeshLocation(
    Bone / Socket Index=$Particles.PartitionParticles.Partition
    Bone / Socket Sampling Mode=::NewEnumerator5
    MaskSpawnGroup=false
    Mesh Sampling Type=::NewEnumerator6
    Skeletal Mesh=$System.SkeletalMesh
   ) {enabled=false}
   &AddVelocity(Velocity=&RandomRangeVector(Minimum=<-1000 0 0>, Maximum=<-1000 0 0>), Velocity Speed Scale=2)
   &InitialMeshOrientation(Use Rotation Vector=false)
   &ShapeLocation(MaskSpawnGroup=false, Override Seed=false, Sphere Radius=&Multiply_Float(A=10, B=4), UseOwnerScale=true)
  ]
  particleUpdateScript [
   &ParticleState()
   &ScaleMeshSize(Scale Factor=&VectorFromCurve(VectorCurve=DataInterface() {propertyValues {XCurve {keys [<at 0.8 1>,<at 1 0>]}, YCurve {keys [<at 0.8 1>,<at 1 0>]}, ZCurve {keys [<at 0.8 1>,<at 1 0>]}}}))
   &InheritSourceMovement(
    Applied Position Scale=&Multiply_VectorByFloat(
     Vector=<1 1 1>
     Float=&FloatFromCurve(FloatCurve=DataInterface() {propertyValues {Curve {keys [
          {arriveTangent=0.0108, interpMode=RCIM_Cubic, leaveTangent=-0.0051, tangentMode=RCTM_Break, time=0, value=1}
          {arriveTangent=-0.2083, interpMode=RCIM_Cubic, tangentMode=RCTM_Break, time=1, value=0.9}
         ]}}})
    )
    ApplyForce=false
    ApplyPosition=true
    ApplyPositionFalloff=false
   )
   &DynamicMaterialParameters(
    Index 0 Param 1=&RandomRangeFloat(Minimum=0, Maximum=1)
    Index 0 Param 2=&FloatFromCurve(FloatCurve=DataInterface() {propertyValues {Curve {keys [<at 0 5>,<at 0.25 5>,<at 1 0>]}}})
    Index 0 Param 3=&FloatFromCurve(FloatCurve=DataInterface() {propertyValues {Curve {keys [<at 0 2>,<at 1 31>]}}})
    Index 0 Param 4=1
    Module.Index0Param0Write=true
    Module.Index0Param1Write=true
    Module.Index0Param2Write=true
    Module.Index0Param3Write=true
   ) {enabled=false}
   &SolveForcesAndVelocity(Clamp Velocity=false, Limit Acceleration=false)
   &DynamicMaterialParameters(
    Index 0 Param 1=1
    Module.Index0Float0Write=true
    Module.Index0Float1Write=true
    Module.Index0Float2Write=true
    Module.Index0Float3Write=true
   ) {module=DynamicMaterialParameters001}
  ]
  rendererClasses [&NiagaraMeshRendererProperties, &NiagaraSpriteRendererProperties]
  renderers [
   {rendererClass=&NiagaraMeshRendererProperties, rendererIndex=0}
   {rendererClass=&NiagaraSpriteRendererProperties, rendererIndex=1}
  ]
 }
 Local_Particles001 {
  emitterUpdateScript [
   &EmitterState()
   &VC_FilteredSocketCount() {enabled=false}
   &SpawnRate(SpawnRate=5, Use Spawn Probability=false)
   &SpawnRate(SpawnGroup=1, SpawnRate=25, Use Spawn Probability=false) {module=SpawnRate001}
   &SpawnRate(SpawnGroup=2, SpawnRate=9, Use Spawn Probability=false) {module=SpawnRate002}
  ]
  particleSpawnScript [
   &VC_PaintedVariantsOverride() {enabled=false}
   /SetVariables(Particles.VisTag:NiagaraInt32=&SpawnGroup()) {isSetParameters=true}
   &InitializeParticle(
    Color Mode=::NewEnumerator0
    Lifetime=&MergeHelper_FloatSwitch(
     0=1
     1=&RandomRangeFloat(Minimum=0.4, Maximum=0.8)
     2=1
     3=?
     Index=$Particles.VisTag
     Num=::NewEnumerator1
    )
    Mesh Scale Mode=::NewEnumerator3
    Mesh Uniform Scale=&RandomRangeFloat(Minimum=0.3, Maximum=0.45)
    Position Offset=&AddVectorToPosition(Position=<50 150 0>, Vector=$User.SingleBoostPos)
    Position Offset Coordinate Space=::Local
    Sprite Rotation Mode=::NewEnumerator2
    Sprite Size Mode=::NewEnumerator3
    Uniform Sprite Size=&MaskFloatBySpawnGroup(Masked Float=80, Passthrough Float=&RandomRangeFloat(Minimum=3, Maximum=6))
    UsePositionOffset=true
    Write Material Random=true
   )
   &PartitionParticles(Number Of Partitions=$Emitter.PortCount)
   &SkeletalMeshLocation(
    Bone / Socket Index=$Particles.PartitionParticles.Partition
    Bone / Socket Sampling Mode=::NewEnumerator5
    MaskSpawnGroup=false
    Mesh Sampling Type=::NewEnumerator6
    Skeletal Mesh=$System.SkeletalMesh
   )
   &AddVelocity(Velocity=<0 0 0>, Velocity Speed Scale=2) {enabled=false}
   &Color(Color=&MergeHelper_ColorSwitch(
     0=<rgba 1 1 1 0.04>
     1=&SelectLinearColorFromArray(
      Array Sampling Mode=::NewEnumerator0
      Direct Array Index=?
      Color Selection Array=DataInterface() {propertyValues {ColorData [<rgba 1 0.05 0 1>, <rgba 0 0.3 1 1>]}}
      Random Seed=0
     )
     2=<rgba 0.6198 0.2879 0 1>
     3=?
     Index=$Particles.VisTag
     Num=::NewEnumerator1
    ))
   &InitialMeshOrientation(Use Rotation Vector=false)
  ]
  particleUpdateScript [
   &ParticleState()
   &SpriteRotationRate(Rotation Rate=&MaskFloatBySpawnGroup(Masked Float=500, Passthrough Float=0)) {module=SpriteRotationRate001}
   &RotateAroundPoint(
    Radius:NiagaraFloat=/VehicleCosmetics/FX/Niagara/Modules/MergeEmitters/VC_FloatBySpawnGroup(
     SpawnGroup or ExecIndex:NiagaraInt32=0
     CustomInt:NiagaraInt32=0
     MaximumSpawnGroups:NiagaraInt32=4
     Float SG 0:NiagaraFloat=0
     Float SG 1:NiagaraFloat=&RandomRangeFloat(Minimum=30, Maximum=40)
     Float SG 2:NiagaraFloat=40
     Float SG 3:NiagaraFloat=0
     Float SG 4:NiagaraFloat=0
     Float SG 5:NiagaraFloat=0
    )
    Rotation Center:NiagaraPosition=&AddVectorToPosition(
     Position=<-95 0 0>
     Vector=&AddVectorToPosition(Position=$User.DTCarOffset, Vector=$User.SingleBoostPos)
    )
    Rotation Phase:NiagaraFloat=$Engine.RealTime
    Rotation Rate:NiagaraFloat=&VC_FloatBySpawnGroup(
     SpawnGroup or ExecIndex:NiagaraInt32=0
     CustomInt:NiagaraInt32=0
     MaximumSpawnGroups:NiagaraInt32=4
     Float SG 0:NiagaraFloat=0
     Float SG 1:NiagaraFloat=0.4545
     Float SG 2:NiagaraFloat=1
     Float SG 3:NiagaraFloat=0
     Float SG 4:NiagaraFloat=0
     Float SG 5:NiagaraFloat=0
    )
    Y Axis:Vector3f=/VehicleCosmetics/FX/Niagara/Modules/MergeEmitters/VC_VectorBySpawnGroup(
     SpawnGroup or ExecIndex:NiagaraInt32=0
     CustomInt:NiagaraInt32=0
     MaximumSpawnGroups:NiagaraInt32=3
     Vector Group 0:Vector3f=<0 0 0>
     Vector Group 1:Vector3f=&RandomRangeVector(Minimum=<-1 -1 -1>, Maximum=<1 1 1>)
     Vector Group 2:Vector3f=<0 1 0>
     Vector Group 3:Vector3f=<0 0 0>
     Vector Group 4:Vector3f=<0 0 0>
     Vector Group 5:Vector3f=<0 0 0>
    )
   )
   &ScaleSpriteSize(
    Scale Sprite Size Mode=::NewEnumerator0
    Uniform Scale Factor=&MaskFloatBySpawnGroup(
     Masked Float=&FloatFromCurve(FloatCurve=DataInterface() {propertyValues {Curve {keys [{tangentMode=RCTM_User, time=0, value=0.5}, {tangentMode=RCTM_User, time=1, value=1.2}]}}})
     Passthrough Float=1
    )
   )
   &ScaleColor(
    Scale Alpha=&MergeHelper_FloatSwitch(
     0=&FloatFromCurve(FloatCurve=DataInterface() {propertyValues {Curve {keys [
          {interpMode=RCIM_Cubic, leaveTangent=2.0041, tangentMode=RCTM_Break, time=0, value=0}
          {interpMode=RCIM_Cubic, leaveTangent=-0.0281, tangentMode=RCTM_Break, time=0.3, value=0.5}
          {arriveTangent=-0.6391, interpMode=RCIM_Cubic, leaveTangent=-0.0249, tangentMode=RCTM_Break, time=1, value=0}
         ]}}})
     1=&FloatFromCurve(
      FloatCurve=DataInterface() {propertyValues {Curve {keys [
          {interpMode=RCIM_Cubic, tangentMode=RCTM_Break, time=0, value=0}
          {interpMode=RCIM_Cubic, tangentMode=RCTM_Break, time=0.3, value=0.5}
          {interpMode=RCIM_Cubic, tangentMode=RCTM_Break, time=1, value=0}
         ]}}}
      Scale Curve=10
     )
     2=1
     3=?
     Index=$Particles.VisTag
     Num=::NewEnumerator1
    )
    Scale RGB=&VectorFromFloat(Value=&MergeHelper_FloatSwitch(
      0=&FloatFromCurve(FloatCurve=DataInterface() {propertyValues {Curve {keys [<at 0 5>,<at 1 1>]}}})
      1=30
      2=1
      3=?
      Index=$Particles.VisTag
      Num=::NewEnumerator1
     ))
    ScaleA=true
    ScaleRGB=true
   )
   &SolveForcesAndVelocity(Clamp Velocity=false, Limit Acceleration=false)
   &DynamicMaterialParameters(
    Index 1 Param 1=&RandomRangeFloat(Minimum=0, Maximum=1)
    Index 1 Param 2=&FloatFromCurve(FloatCurve=DataInterface() {propertyValues {Curve {keys [<at 0 5>,<at 0.25 5>,<at 1 0>]}}})
    Index 1 Param 3=&FloatFromCurve(FloatCurve=DataInterface() {propertyValues {Curve {keys [<at 0 2>,<at 1 31>]}}})
    Index 1 Param 4=1
    Module.Index1Float0Write=true
    Module.Index1Float1Write=true
    Module.Index1Float2Write=true
    Module.Index1Float3Write=true
    Write Parameter Index 0=::NewEnumerator0
    Write Parameter Index 1=::NewEnumerator1
   )
   /SetVariables(Particles.DMP_Stone:Vector4f=&MakeVector4(X=1, Y=0, Z=0, W=0)) {isSetParameters=true}
  ]
  rendererClasses [&NiagaraSpriteRendererProperties, &NiagaraMeshRendererProperties]
  renderers [
   {rendererClass=&NiagaraSpriteRendererProperties}
   {rendererClass=&NiagaraMeshRendererProperties, rendererIndex=1}
   {rendererClass=&NiagaraSpriteRendererProperties, rendererIndex=2}
  ]
 }
 Ribbons001 {
  emitterUpdateScript [
   &VC_RateQualityHandler(Epic Spawn Rate=30, High Spawn Rate=28, Low Spawn Rate=20, Medium Spawn Rate=25, Use Custom Rate?=true) {module=VC_RateQualityHandler001}
   &EmitterState()
   &VC_FilteredSocketCount() {enabled=false}
   &VC_RateQualityHandler(Epic Spawn Rate=20, High Spawn Rate=20, Low Spawn Rate=20, Medium Spawn Rate=20, Use Custom Rate?=true)
   &VC_SpawnRibbon_v1(Number of Ribbons=2, SpawnRate=$Emitter.Spawn Rate)
  ]
  particleSpawnScript [
   &VC_PaintedVariantsOverride() {enabled=false}
   &VC_InitializeRibbon_v1(
    Color=<rgba 0.0466 0.108 1 1>
    Custom Offset=true
    Lifetime=1
    Local Vector Offset=&AddVectorToPosition(
     Position=<-60 0 0>
     Vector=&AddVectorToPosition(Position=$User.DTCarOffset, Vector=$User.SingleBoostPos)
    )
    Number of Bones Per Boneflow group=0
    Offset Amount=1
    Ribbon Width:NiagaraFloat=&Multiply_Float(A=5, B=4)
   )
   &AddVelocity(Velocity=<-2000 0 0>) {module=AddVelocity001}
   /VehicleCosmetics/FX/Niagara/Modules/MergeEmitters/VC_ExtraBindingColors(ExtraColors=2, R1 Color:LinearColor=<rgba 1 0.03 0 1>, R2 Color:LinearColor=<rgba 0.8802 0.9615 1 0.2>)
   /SetVariables(Particles.Wind_Width=100) {isSetParameters=true}
  ]
  particleUpdateScript [
   &ParticleState()
   &ScaleColor(
    Scale Alpha=&FloatFromCurve(FloatCurve=DataInterface() {propertyValues {Curve {keys [<at 0 1 RCIM_Cubic>, {arriveTangent=-6.4238, interpMode=RCIM_Cubic, leaveTangent=-6.4238, tangentMode=RCTM_User, time=0.5, value=0}]}}})
    Scale RGB=&MakeCustomVectorFromBool(Bool=$User.bIsFrontendPreview, True Vector=<3 3 3>, False Vector=<1 1 3>)
    ScaleA=true
    ScaleRGB=true
   )
   &InheritSourceMovement(
    Applied Position Scale=&Multiply_VectorByFloat(
     Vector=<1 1 1>
     Float=&FloatFromCurve(FloatCurve=DataInterface() {propertyValues {Curve {keys [
          {interpMode=RCIM_Cubic, leaveTangent=-0.0051, tangentMode=RCTM_Break, time=0, value=1}
          {arriveTangent=-0.0476, interpMode=RCIM_Cubic, tangentMode=RCTM_Break, time=1, value=0.9}
         ]}}})
    )
    ApplyForce=false
    ApplyPosition=true
    ApplyPositionFalloff=false
   )
   &DynamicMaterialParameters(
    Index 0 Param 1=3
    Module.Index0Param0Write=true
    Module.Index0Param1Write=true
    Module.Index0Param2Write=true
    Module.Index0Param3Write=true
   )
   &SolveForcesAndVelocity(Clamp Velocity=false, Limit Acceleration=false)
   &VC_UpdateRibbon_v1()
   /SetVariables(Particles.R1 Color:LinearColor=&VC_DynamicScaleColor(
     Color to Scale:LinearColor=$Particles.Initial.R1 Color
     RGB Scale:Vector3f=&VectorFromCurve(
      VectorCurve=DataInterface() {propertyValues {XCurve {keys [<at 0 1>,<at 1 0>]}, YCurve {keys [<at 0 1>,<at 1 0>]}, ZCurve {keys [<at 0 1>,<at 1 0>]}}}
      Scale Curve=&MakeCustomVectorFromBool(Bool=$User.bIsFrontendPreview, True Vector=<5 5 5>, False Vector=<5 2 2>)
     )
     Alpha Scale:NiagaraFloat=&FloatFromCurve(
      FloatCurve=DataInterface() {propertyValues {Curve {keys [
          {arriveTangent=-2.2222, interpMode=RCIM_Cubic, time=0.04, value=0}
          {interpMode=RCIM_Cubic, tangentMode=RCTM_User, time=0.05, value=1}
          {arriveTangent=-6.4238, interpMode=RCIM_Cubic, leaveTangent=-6.4238, tangentMode=RCTM_User, time=0.5, value=0}
         ]}}}
      Scale Curve=3
     )
    )) {isSetParameters=true}
  ]
  rendererClasses [&NiagaraRibbonRendererProperties]
  renderers [
   {rendererClass=&NiagaraRibbonRendererProperties}
   {rendererClass=&NiagaraRibbonRendererProperties, rendererIndex=1}
   {rendererClass=&NiagaraRibbonRendererProperties, rendererIndex=2}
   {rendererClass=&NiagaraRibbonRendererProperties, rendererIndex=3}
  ]
 }
 Ring_Water_Fire001 {
  emitterUpdateScript [
   &EmitterState(Inactive Response=::NewEnumerator1, Life Cycle Mode=::NewEnumerator1, Loop Duration=1, UseLoopDelay=false)
   &SpawnBurst_Instantaneous(Use Spawn Probability=false, UseLoopCountLimit=false)
   &VC_FilteredSocketCount() {enabled=false}
   &SpawnRate(SpawnRate=5, Use Spawn Probability=false) {enabled=false}
  ]
  particleSpawnScript [
   &VC_PaintedVariantsOverride() {enabled=false}
   &InitializeParticle(
    Color=<rgba 0.1016 0.1916 1.5 1>
    Lifetime=1.01
    Mesh Scale Mode=::NewEnumerator3
    Mesh Uniform Scale=&Multiply_Float(A=0.12, B=2.5)
    Position Offset=&AddVectorToPosition(
     Position=&AddVectorToPosition(
      Position=<40 0 30>
      Vector=&AddVectorToPosition(Position=$User.DTCarOffset, Vector=$User.SingleBoostPos)
     )
     Vector=<-135 0 -30>
    )
    Position Offset Coordinate Space=::Local
    UsePositionOffset=true
    Write Material Random=true
   )
   &InitialMeshOrientation(Mesh Orientation Mode=::NewEnumerator7, Rotation=<-0.6296 0 0>, Use Rotation Vector=true)
   &PartitionParticles(Number Of Partitions=$Emitter.PortCount) {enabled=false}
   &SkeletalMeshLocation(
    Bone / Socket Index=$Particles.PartitionParticles.Partition
    Bone / Socket Sampling Mode=::NewEnumerator5
    MaskSpawnGroup=false
    Mesh Sampling Type=::NewEnumerator6
    Skeletal Mesh=$System.SkeletalMesh
   ) {enabled=false}
   &AddVelocity(Velocity=&RandomRangeVector(Minimum=<-1000 0 0>, Maximum=<-1000 0 0>), Velocity Speed Scale=2) {enabled=false}
   /SetVariables(Particles.Color_Fire:LinearColor=<rgba 1 0.05 0 1>, Particles.Or:Quat4f=<0.6296 0 0 1>) {isSetParameters=true}
  ]
  particleUpdateScript [
   &ParticleState()
   &ScaleColor(
    Scale RGB=&MakeCustomVectorFromBool(
     Bool=$User.bIsFrontendPreview
     True Vector=&VectorFromCurve(VectorCurve=DataInterface() {propertyValues {
        XCurve {keys [{tangentMode=RCTM_User, time=0, value=2}, {tangentMode=RCTM_User, time=1, value=1}]}
        YCurve {keys [{tangentMode=RCTM_User, time=0, value=2}, {tangentMode=RCTM_User, time=1, value=1}]}
        ZCurve {keys [<at 0 2>,<at 1 1>]}
       }})
     False Vector=&VectorFromCurve(VectorCurve=DataInterface() {propertyValues {XCurve {keys [<at 0 2>,<at 1 1>]}, YCurve {keys [<at 0 2>,<at 1 1>]}, ZCurve {keys [<at 0 2>,<at 1 1>]}}})
    )
    ScaleA=true
    ScaleRGB=true
   )
   &SolveForcesAndVelocity(Clamp Velocity=false, Limit Acceleration=false)
  ]
  rendererClasses [&NiagaraMeshRendererProperties, &NiagaraSpriteRendererProperties]
  renderers [
   {rendererClass=&NiagaraMeshRendererProperties, rendererIndex=0}
   {rendererClass=&NiagaraMeshRendererProperties, rendererIndex=1}
   {rendererClass=&NiagaraSpriteRendererProperties, rendererIndex=2}
  ]
 }
}
summary {
 systemName=NS_Boost_Elements_Optimized
 emitters [
  {emitterName=Ribbons001, rendererClasses [&NiagaraRibbonRendererProperties]}
  {
   emitterName=Ring_Water_Fire001
   rendererClasses [&NiagaraMeshRendererProperties, &NiagaraSpriteRendererProperties]
  }
  {
   emitterName=Flying_Rocks001
   rendererClasses [&NiagaraMeshRendererProperties, &NiagaraSpriteRendererProperties]
  }
  {
   emitterName=Local_Particles001
   rendererClasses [&NiagaraSpriteRendererProperties, &NiagaraMeshRendererProperties]
  }
 ]
 userVariables [
  User.bIsFrontendPreview {type=NiagaraBool}
  User.bIsPainted {type=NiagaraBool}
  User.ColorVariantID {type=NiagaraInt32}
  User.DTCarOffset {type=NiagaraPosition, description="", defaultValue {}}
  User.PaintedColor {type=LinearColor, description="", defaultValue {struct=LinearColor, value=<rgba 1 1 1 1>}}
  User.PaintedScalar {type=NiagaraFloat, defaultValue=1, description=""}
  User.PortCount {type=NiagaraInt32, defaultValue=1}
  User.SingleBoostPos {type=Vector3f, description="", defaultValue {struct=Vector3f, value=<0 0 0>}}
 ]
}
