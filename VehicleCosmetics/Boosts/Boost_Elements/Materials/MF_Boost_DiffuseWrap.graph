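# MF_Boost_DiffuseWrap is a material function for vehicle boost materials that computes wrapped diffuse lighting from an N·L input and a WrapAmount scalar.
# It adds WrapAmount to N·L, divides by one plus WrapAmount and clamps the result.
# This wrap lighting technique lets diffuse light bleed past the terminator for a softer falloff.
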MaterialFunction /VehicleCosmetics/Boosts/Boost_Elements/Materials/MF_Boost_DiffuseWrap
graph {nodes [
  n1 = FunctionInput() {inputName=WrapAmount, inputType=FunctionInput_Scalar}
  n2 = Add(A=FunctionInput() {inputName=NDotl, inputType=FunctionInput_Scalar}, B=@n1)
  n3 = Add(A=@n1, B=Constant() {r=1})
  FunctionOutput(None=Clamp(None=Divide(A=@n2, B=@n3)))
 ]}
